WidgetBlueprint /Sidebar/Panels/Overlay/WBP_Confirmation_RewardsList
graphs [
 EventGraph {links [], nodes []}
 Get Reward Items {nodes [
   Get Reward Items() {is=title K2Node_FunctionEntry}
   > Return Node!(
    execute
    Item Definitions:Array of Fort Item Definition Object References=Get Reward Item Definitions!(self:WBP Confirmation Rewards List Entry Container Object Reference=Get RewardsListEntryContainer() -> (RewardsListEntryContainer:WBP Confirmation Rewards List Entry Container Object Reference)) -> (Reward Item Definitions:Array of Fort Item Definition Object References)
   )
  ]}
 Update {nodes [
   Update!() -> (
    then
    Mission Rewards:Array of Fort Mission Reward Info Structures
    Gameplay Modifiers:Array of Fort Gameplay Modifier Item Definition Object References
   ) {is=title K2Node_FunctionEntry}
   > n1 = Sequence()
   > Update!(
    execute
    self:WBP Confirmation Rewards List Entry Container Object Reference=Get RewardsListEntryContainer()
    Mission Rewards:Array of Fort Mission Reward Info Structures=Get Mission Rewards() -> (Mission Rewards:Array of Fort Mission Reward Info Structures)
    Gameplay Modifiers:Array of Fort Gameplay Modifier Item Definition Object References=Get Gameplay Modifiers() -> (Gameplay Modifiers:Array of Fort Gameplay Modifier Item Definition Object References)
   ) -> (then) {is=title K2Node_CallFunction}
   n2 = Branch(execute=@n1.then_1, Condition=Get ShowMissionAlertMessage() -> (ShowMissionAlertMessage:bool))
   n3 = GetActiveTile!(self:Fort Front End Context Object Reference=Get FortFrontEndContext!() -> (ReturnValue:Fort Front End Context Object Reference) {is=title K2Node_GetBlueprintContextEx}) -> (ReturnValue:Fort Theater Map Tile Object Reference)
   > n4 = Branch(Condition=IsMissionAlertTile!(self:Fort Theater Map Tile Object Reference=@n3) -> (ReturnValue:bool))
   n5 = Get FortFrontEndContext()
   > SetVisibility(
    self=Get BorderMissionAlert() -> (BorderMissionAlert:Common Border Object Reference)
    InVisibility=Select(
     Option 0=SelfHitTestInvisible
     Option 1=Collapsed
     Index=IsTileMissionAlertRewardClaimed!(self:Fort Front End Context Object Reference=@n5, Tile:Fort Theater Map Tile Object Reference=@n3) -> (ReturnValue:bool)
    )
   )
   SetVisibility(execute=[@n4.else, @n2.else], self=Get BorderMissionAlert(), InVisibility=Collapsed)
  ]}
]
variables [
 ShowMissionAlertMessage {type=boolean}
 Use Verbose Mode {type=boolean}
]
widgets {
 info {inheritedWidgetCount=0, parentClass=&CommonUserWidget, rootWidgetClass=&VerticalBox, widgetCount=12}
 widgets [
  VerticalBox_1 {widgetClassPath=&VerticalBox, widget=VerticalBox_1}
  HorizontalBox_0 {
   widgetClassPath=&HorizontalBox
   parent=VerticalBox_1
   widget=HorizontalBox_0
   properties {flowDirectionPreference=Culture}
   slot {class=VerticalBoxSlot}
  }
  HorizontalBox_2 {
   widgetClassPath=&HorizontalBox
   parent=HorizontalBox_0
   widget=HorizontalBox_2
   slot {class=HorizontalBoxSlot, properties {horizontalAlignment=HAlign_Left, size {sizeRule=Fill, value=1}}}
  }
  SizeBox_0 {
   widgetClassPath=&SizeBox
   parent=HorizontalBox_2
   widget=SizeBox_0
   properties {bOverride_HeightOverride=true, heightOverride=40}
   slot {class=HorizontalBoxSlot, properties {padding=<ltrb 32 0 48 0>}}
  }
  CommonTextBlock_0 {
   widgetClassPath=&CommonTextBlock
   bIsVariable=true
   parent=SizeBox_0
   widget=CommonTextBlock_0
   properties {
    style=/FortUILibrary/Styles/Typography/Foundation/TextStyle-Base-S-B-60pc.TextStyle-Base-S-B-60pc_C
    text=REWARDS YOU'LL BE MISSING
    colorAndOpacity {colorUseRule=UseColor_Specified, specifiedColor=<rgba 1 1 1 0.6>}
    font {
     bForceMonospaced=false
     bMaterialIsStencil=false
     fontMaterial=null
     fontObject=&NotoSans
     letterSpacing=0
     monospacedWidth=1
     size=18
     skewAmount=0
     typefaceFontName=NotoSans-Bold
     outlineSettings {
      bApplyOutlineToDropShadows=false
      bMiteredCorners=false
      bSeparateFillAlpha=false
      outlineMaterial=null
      outlineSize=0
      outlineColor=<rgba 0 0 0 1>
     }
    }
    shadowOffset=<0 0>
   }
   slot {class=SizeBoxSlot, properties {verticalAlignment=VAlign_Bottom}}
  }
  Spacer_1 {widgetClassPath=&Spacer, parent=VerticalBox_1, widget=Spacer_1, properties {size=<1 8>}, slot {class=VerticalBoxSlot}}
  VerticalBox_0 {
   widgetClassPath=&VerticalBox
   parent=VerticalBox_1
   widget=VerticalBox_0
   slot {class=VerticalBoxSlot, properties {size {sizeRule=Fill, value=1}}}
  }
  RewardsListEntryContainer {
   widgetClassPath=&/WBP_Confirmation_RewardsListEntryContainer
   bIsVariable=true
   parent=VerticalBox_0
   widget=RewardsListEntryContainer
   slot {class=VerticalBoxSlot, properties {padding=<ltrb 16 0 16 16>, size {sizeRule=Fill, value=1}}}
  }
  BorderMissionAlert {
   widgetClassPath=&CommonBorder
   bIsVariable=true
   parent=VerticalBox_0
   widget=BorderMissionAlert
   properties {style=/FortUILibrary/Styles/Border/Border-SolidBG-Yellow.Border-SolidBG-Yellow_C, brushColor=<rgba 1 1 1 0.3>, padding=<ltrb 10 10 10 10>}
   slot {class=VerticalBoxSlot, properties {verticalAlignment=VAlign_Bottom}}
  }
  HorizontalBox_1 {
   widgetClassPath=&HorizontalBox
   parent=BorderMissionAlert
   widget=HorizontalBox_1
   properties {flowDirectionPreference=Culture}
   slot {class=BorderSlot, properties {padding=<ltrb 10 10 10 10>}}
  }
  Image_0 {
   widgetClassPath=&Image
   bIsVariable=true
   parent=HorizontalBox_1
   widget=Image_0
   properties {brush {
     bIsDynamicallyLoaded=false
     drawAs=Image
     imageType=NoImage
     mirroring=NoMirror
     resourceName=None
     resourceObject=/Game/UI/Foundation/Textures/Icons/HUD/T-Icon-Clock-128
     tiling=NoTile
     imageSize=<32 32>
     margin=<ltrb 0 0 0 0>
     outlineSettings {
      bUseBrushTransparency=false
      roundingType=HalfHeightRadius
      width=0
      color {colorUseRule=UseColor_Specified, specifiedColor=<rgba 0 0 0 0>}
      cornerRadii=<0 0 0 0>
     }
     tintColor {colorUseRule=UseColor_Specified, specifiedColor=<rgba 1 0.8714 0.0723 1>}
     uVRegion {bIsValid=false, max=<0 0>, min=<0 0>}
    }}
   slot {class=HorizontalBoxSlot, properties {horizontalAlignment=HAlign_Left, verticalAlignment=VAlign_Center, padding=<ltrb 0 0 16 0>}}
  }
  CommonTextBlock_1 {
   widgetClassPath=&CommonTextBlock
   parent=HorizontalBox_1
   widget=CommonTextBlock_1
   properties {
    autoWrapText=true
    style=/FortUILibrary/Styles/Typography/Foundation/TextStyle-Base-XS-B-Yellow.TextStyle-Base-XS-B-Yellow_C
    text="This time-limited mission has a special reward!"
    colorAndOpacity {colorUseRule=UseColor_Specified, specifiedColor=<rgba 1 0.9301 0 1>}
    font {
     bForceMonospaced=false
     bMaterialIsStencil=false
     fontMaterial=null
     fontObject=&NotoSans
     letterSpacing=0
     monospacedWidth=1
     size=16
     skewAmount=0
     typefaceFontName=NotoSans-Bold
     outlineSettings {
      bApplyOutlineToDropShadows=false
      bMiteredCorners=false
      bSeparateFillAlpha=false
      outlineMaterial=null
      outlineSize=0
      outlineColor=<rgba 0 0 0 1>
     }
    }
    shadowOffset=<0 0>
   }
   slot {class=HorizontalBoxSlot, properties {size {sizeRule=Fill, value=1}}}
  }
 ]
}
