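# Sidebar "Add friends" panel widget for searching players by display name, browsing suggested and recent players, and sharing friend invites.
# Enter runs a search through FortSocialPlayerSearchVM; results fill the social tree list, with a throbber during searches.
# Empty text restores default headers, chosen by the Epic parties flag.
# Item clicks open profiles, blocking goes through a confirmation panel, and social bans swap search and invite controls for a message.
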WidgetBlueprint /Sidebar/Panels/WBP_AddFriends_Panel
graphs [
 BP_GetDesiredFocusTarget {nodes [
   Get Desired Focus Target()
   > n1 = Branch(Condition=IsVisible(self=Get WBP_Sidebar_FilteredInputField() -> (WBP_Sidebar_FilteredInputField:WBP Sidebar Filtered Input Field Object Reference)))
   n2 = Get WBP_Sidebar_InputField!(self:WBP Sidebar Filtered Input Field Object Reference=Get WBP_Sidebar_FilteredInputField()) -> (WBP_Sidebar_InputField:WBP Sidebar Input Field Object Reference)
   > Return Node!(
    execute
    ReturnValue:Widget Object Reference=Get EditableText!(self:WBP UIKit Block Editable Text Object Reference=Get Block_EditableText!(self:WBP UIKit Input Field Base Object Reference=@n2) -> (Block_EditableText:WBP UIKit Block Editable Text Object Reference)) -> (EditableText:UIKit Editable Text Object Reference)
   )
   Return Node(execute=@n1.else, ReturnValue=Get WBP_Sidebar_SocialTreeList() -> (WBP_Sidebar_SocialTreeList:WBP Sidebar Social Tree List Object Reference))
  ]}
 EventGraph {nodes [
   n1 = "Item Clicked (WBP_Sidebar_SocialTreeList)"!() -> (OutputDelegate, then, ItemVM:Fort Social Item VM Object Reference) {is=title K2Node_ComponentBoundEvent}
   > NavigateToPlayerProfile!(
    execute
    SelfReference:Object Reference=Self-Reference()
    SocialItemVM:Fort Social Item VM Object Reference=@n1.ItemVM
    SidebarNavigationVM:Fort Sidebar Navigation VM Object Reference=Get FortSidebarNavigationVM() -> (FortSidebarNavigationVM:Fort Sidebar Navigation VM Object Reference)
    Sidebar VM:Fort Sidebar VM Object Reference
   ) -> (then)
   "On Clicked (WBP_Sidebar_InviteAFriend)"!() -> (OutputDelegate, then, Button:Common Button Base Object Reference) {is=title K2Node_ComponentBoundEvent}
   > OpenFriendsInviteShareUrlDialog!(execute, Context:Widget Object Reference, AnalyticsContext:String=Sidebar) -> (then)
   "On Clear Button Clicked (WBP_Sidebar_FilteredInputField)"() {is=title K2Node_ComponentBoundEvent}
   > RequestRefreshFocus(self:Common Activatable Widget Object Reference)
   n2 = "On Text Changed (WBP_Sidebar_FilteredInputField)"!() -> (OutputDelegate, then, Text:Text) {is=title K2Node_ComponentBoundEvent}
   > Branch(Condition=TextIsEmpty(InText=@n2.Text))
   > Populate Social Headers()
   n3 = "On Text Committed (WBP_Sidebar_FilteredInputField)"!() -> (OutputDelegate, then, CommitMethod:ETextCommit Enum=Default) {is=title K2Node_ComponentBoundEvent}
   > n4 = Switch on ETextCommit!(execute, Selection:ETextCommit Enum=@n3.CommitMethod) -> (Default:Exec, OnEnter:Exec, OnUserMovedFocus:Exec, OnCleared:Exec) {is=title K2Node_SwitchEnum}
   n5 = GetCurrentText!(execute=@n4.OnEnter, self:WBP Sidebar Filtered Input Field Object Reference=Get WBP_Sidebar_FilteredInputField()) -> (then, Text:Text)
   > Init(self:Fort Social Player Search VM Object Reference=Get FortSocialPlayerSearchVM() -> (FortSocialPlayerSearchVM:Fort Social Player Search VM Object Reference))
   > SearchPlayers!(execute, self:Fort Social Player Search VM Object Reference=Get FortSocialPlayerSearchVM(), SearchTerm:String="To String (Text)"(InText=@n5)) -> (then)
   > RequestRefreshFocus()
   n6 = "Item Block Player (WBP_Sidebar_SocialTreeList)"!() -> (OutputDelegate, then, SocialUser:Fort Social User Object Reference, Interaction:Fort Social Interaction VM Object Reference) {is=title K2Node_ComponentBoundEvent}
   > Set BlockPlayerInteraction(BlockPlayerInteraction:Fort Social Interaction VM Object Reference=@n6.Interaction) -> (Output_Get:Fort Social Interaction VM Object Reference)
   > Navigate to Block Player!(exec, SocialUser:Fort Social User Object Reference=@n6.SocialUser) -> (then:Exec) {is=title K2Node_MacroInstance}
   Event PreConstruct()
   > SetMobileStyle() {is=title K2Node_CallFunction}
   > SetSize!(
    execute
    self:WBP UIKit Input Field Base Object Reference=Get WBP_Sidebar_InputField(self=Get WBP_Sidebar_FilteredInputField())
    WIdth:S UI Bool Float Structure
    Height_Enable_3_#1:bool=true
    Height_Value_4_#2:double=Get InputFieldHeight() -> (InputFieldHeight:double)
   ) -> (then)
   > SetHideInputAction!(
    execute
    self:Common Button Base Object Reference=Get WBP_Sidebar_InputField(self=Get WBP_Sidebar_FilteredInputField())
    bInHideInputAction:bool=false
   ) -> (then)
   Event Construct()
   > Set with Broadcast ShowBackButton!(
    execute
    ShowBackButton:bool=isUseEpicParties!() -> (UseEpicParties:bool)
    self:WBP Sidebar Panel Layout Object Reference=Get WBP_Sidebar_PanelLayout() -> (WBP_Sidebar_PanelLayout:WBP Sidebar Panel Layout Object Reference)
   ) -> (then, Output_Get:bool)
   > Populate Social Headers()
   > Branch(Condition=GetConfigDataBoolWithDefault!(EntryName:String=social_friend_invite_alt_text_shareinvite, bDefaultValue:bool=false) -> (ReturnValue:bool))
   > SetText!(
    execute
    self:WBP UIKit Button Generic Base Object Reference=Get WBP_Sidebar_InviteAFriend() -> (WBP_Sidebar_InviteAFriend:WBP Sidebar Button Generic Object Reference)
    Text:Text="LOCTEXT(\"Share invite\")"
   ) -> (then)
   Event On Activated()
   > SidebarFeatureFinishedLoading(self:Fort Sidebar VM Object Reference=Get FortSidebarVM() -> (FortSidebarVM:Fort Sidebar VM Object Reference))
  ]}
 HandleConfirmationOverlayActions {nodes [
   n1 = HandleConfirmationOverlayActions!() -> (then, ActionID:Name) {is=title K2Node_FunctionEntry}
   > Allow Handle Overlay Confirmation!(Exec:Exec, Action ID:Name=@n1) -> (then:Exec) {is=title K2Node_MacroInstance}
   > Branch(Condition="Equal (Name)"(A=@n1, B=Block))
   > ExecuteInteraction(self:Fort Social Interaction VM Object Reference=Get BlockPlayerInteraction() -> (BlockPlayerInteraction:Fort Social Interaction VM Object Reference))
  ]}
 NavigateToBlockPlayer {nodes [
   n1 = Inputs!() -> (exec:Exec, SocialUser:Fort Social User Object Reference)
   > n2 = Construct Object from Class(Class=/Sidebar/Panels/Overlay/Data/ConfirmationVM_BlockPlayer.ConfirmationVM_BlockPlayer_C) -> (ReturnValue:Confirmation VM Block Player Object Reference)
   > Set Description!(
    execute
    Description:Text=Format Text!(
     Format="LOCTEXT(\"Would you like to block {PlayerName}?\")"
     PlayerName:String=GetNicknameOrDisplayName!(self:Fort Social User Object Reference=@n1) -> (ReturnValue:String)
    ) -> (Result)
    self:Sidebar Confirmation Panel VM Object Reference=@n2
   ) -> (then, Output_Get:Text)
   > NavigateToConfirmPanel!(
    execute
    SelfReference:Object Reference=Self-Reference()
    SidebarConfirmationVM:Sidebar Confirmation Panel VM Object Reference=@n2
    SidebarNavigationVM:Fort Sidebar Navigation VM Object Reference=Get FortSidebarNavigationVM()
   ) -> (then)
   > Outputs!(then:Exec)
  ]}
 Populate Social Headers {nodes [
   Populate Social Headers() {is=title K2Node_FunctionEntry}
   > ClearHeaders(self:Fort Social Header List VM Object Reference=Get FortSocialPlayerSearchVM())
   > n1 = For Each Loop(Array:Array of ESocialHeaderItemType Enums=Select(
     Option 0:Array of ESocialHeaderItemType Enums=Get Social Headers() -> (Social Headers:Array of ESocialHeaderItemType Enums=Unknown)
     Option 1:Array of ESocialHeaderItemType Enums=Get Social Headers_v2() -> (Social Headers_v2:Array of ESocialHeaderItemType Enums=Unknown)
     Index=isUseEpicParties()
    ) -> (ReturnValue:Array of ESocialHeaderItemType Enums)) -> (Array Element:ESocialHeaderItemType Enum)
   > AddHeader!(execute, self:Fort Social Header List VM Object Reference=Get FortSocialPlayerSearchVM(), HeaderType:ESocialHeaderItemType Enum=@n1) -> (then, ReturnValue:Fort Social Header Item VM Object Reference)
  ]}
 Search Results {nodes [
   n1 = Search Results!() -> (then, OuterArray:Array of Fort Social Header Item VM Object References) {is=title K2Node_FunctionEntry}
   > SetTreeViewItems!(
    execute
    self:WBP Sidebar Social Tree List Object Reference=Get WBP_Sidebar_SocialTreeList()
    OuterArray:Array of Fort Social Header Item VM Object References=@n1
   ) -> (then)
  ]}
 Search Status {nodes [
   n1 = Search Status!() -> (then, IsSearchInProgress:bool=~) {is=title K2Node_FunctionEntry}
   > SetActiveWidget(
    self=Get VisibilitySwitcher() -> (VisibilitySwitcher:Common Visibility Switcher Object Reference)
    Widget=Select(
     Option 0:Widget Object Reference=Get WBP_Sidebar_SocialTreeList()
     Option 1:Widget Object Reference=Get HB_Loading() -> (HB_Loading:Horizontal Box Object Reference)
     Index=@n1
    ) -> (ReturnValue:Widget Object Reference)
   )
  ]}
 SetMobileStyle {nodes [
   SetMobileStyle() {is=title K2Node_FunctionEntry}
   > Branch(Condition=IsUsingCompactUI())
   > Set InputFieldHeight(InputFieldHeight:double=80) -> (Output_Get:double)
   n1 = SlotAsGridSlot(Widget=Get WBP_Sidebar_InviteAFriend())
   n2 = Get!(self:Grid Slot Object Reference=@n1) -> (Padding_Left:float=~, Padding_Top:float=~, Padding_Right:float=~, Padding_Bottom:float=~)
   > SetPadding!(
    execute
    self:Grid Slot Object Reference=@n1
    InPadding_Left:float=@n2
    InPadding_Top:float=16
    InPadding_Right:float=@n2.Padding_Right
    InPadding_Bottom:float=@n2.Padding_Bottom
   ) -> (then)
   n3 = SlotAsOverlaySlot(Widget=Get WBP_Sidebar_SocialTreeList())
   n4 = Get(self:Overlay Slot Object Reference=@n3) -> (Padding_Left=~, Padding_Top=~, Padding_Right=~, Padding_Bottom=~)
   > SetPadding(
    self:Overlay Slot Object Reference=@n3
    InPadding_Left=@n4
    InPadding_Top=20
    InPadding_Right=@n4.Padding_Right
    InPadding_Bottom=@n4.Padding_Bottom
   )
   n5 = SlotAsGridSlot(Widget=Get VisibilitySwitcher())
   n6 = Get(self:Grid Slot Object Reference=@n5) -> (Padding_Left=~, Padding_Top=~, Padding_Right=~, Padding_Bottom=~)
   > SetPadding(
    self:Grid Slot Object Reference=@n5
    InPadding_Left=@n6
    InPadding_Top=24
    InPadding_Right=@n6.Padding_Right
    InPadding_Bottom=@n6.Padding_Bottom
   )
  ]}
 SetSocialBanVisibility {nodes [
   SetSocialBanVisibility!() -> (then, IsSocialBanned:bool=~) {is=title K2Node_FunctionEntry}
   > SetVisibility(
    self=Get WBP_Sidebar_FilteredInputField()
    InVisibility=Select(
     Option 0:ESlateVisibility Enum=SelfHitTestInvisible
     Option 1:ESlateVisibility Enum=Collapsed
     Index=Get IsSocialBanned() -> (IsSocialBanned:bool)
    ) -> (ReturnValue:ESlateVisibility Enum)
   )
   > SetVisibility(
    self=Get WBP_Sidebar_SocialBan_Message() -> (WBP_Sidebar_SocialBan_Message:WBP Social Ban Message Object Reference)
    InVisibility=Select(Option 0=Collapsed, Option 1=SelfHitTestInvisible, Index=Get IsSocialBanned())
   )
   n1 = Should Show Invite A Friend Button!() -> (ReturnValue:bool)
   n2 = NOT Boolean(A=Get IsSocialBanned())
   > SetVisibility(
    self=Get WBP_Sidebar_InviteAFriend()
    InVisibility=Select(Option 0=Collapsed, Option 1=SelfHitTestInvisible, Index=AND Boolean(A=@n2, B=@n1))
   )
   > RequestRefreshFocus()
  ]}
]
variables [
 BlockPlayerInteraction {type=FortSocialInteractionVM, default=null}
 InputFieldHeight {type=number, default=60}
 Social Headers {type="array<ESocialHeaderItemType>", default [FriendInvites, SuggestedFriends, RecentPlayers]}
 Social Headers_v2 {type="array<ESocialHeaderItemType>", default [SuggestedFriends, RecentPlayers]}
]
widgets {
 info {
  inheritedWidgetCount=0
  namedSlotCount=1
  parentClass=&CommonActivatableWidget
  rootWidgetClass=/Sidebar/Shared/WBP_Sidebar_PanelLayout.WBP_Sidebar_PanelLayout_C
  widgetCount=10
 }
 widgets [
  WBP_Sidebar_PanelLayout {
   widgetClassPath=/Sidebar/Shared/WBP_Sidebar_PanelLayout
   bIsVariable=true
   widget=WBP_Sidebar_PanelLayout
   properties {headingText=Add friends}
  }
  AddFriends_Grid_Root {
   widgetClassPath=&GridPanel
   namedSlotHost=WBP_Sidebar_PanelLayout
   widget=AddFriends_Grid_Root
   properties {columnFill [1], rowFill [0,0,0,1]}
  }
  WBP_Sidebar_FilteredInputField {
   widgetClassPath=/Sidebar/Buttons/WBP_Sidebar_FilteredInputField
   bIsVariable=true
   parent=AddFriends_Grid_Root
   widget=WBP_Sidebar_FilteredInputField
   properties {
    bIsFocusable=true
    hideValidationInfo=true
    hintText=Search by display name
    isSearchField=true
    useHintText=true
    clearButtonPadding=<ltrb 0 0 12 0>
    desiredImageSize=<28 28>
    icon Properties {
     material=/FortUILibrary/Systems/InputField/MaterialInstances/MI_UIKit_InputFieldIcon
     image Size=<32 32>
     placement {
      horizontal Alignment=HAlign_Left
      vertical Alignment=VAlign_Center
      grid {column=0, column Span=0, layer=0, row=1, row Span=0, nudge=<0 0>}
      padding=<ltrb 18 0 0 0>
     }
    }
    inputActionPromptPadding=<ltrb 0 0 12 0>
    scaleOverride {enable=false, value=1}
   }
   slot {class=GridSlot, properties {row=1, verticalAlignment=VAlign_Top, padding=<ltrb 18 12 18 0>}}
  }
  WBP_Sidebar_SocialBan_Message {
   widgetClassPath=/FortniteCoreUI/UI/Social/SocialBan/WBP_SocialBan_Message
   bIsVariable=true
   parent=AddFriends_Grid_Root
   widget=WBP_Sidebar_SocialBan_Message
   properties {useFriendsSpecificMessage=true}
   slot {class=GridSlot, properties {layer=1, padding=<ltrb 18 0 18 12>}}
  }
  WBP_Sidebar_InviteAFriend {
   widgetClassPath=/Sidebar/Buttons/WBP_Sidebar_ButtonGeneric
   bIsVariable=true
   parent=AddFriends_Grid_Root
   widget=WBP_Sidebar_InviteAFriend
   properties {
    buttonType=Attention
    minHeightOverride=60
    text=Invite a friend
    font {
     bForceMonospaced=false
     bMaterialIsStencil=false
     fontMaterial=/Sidebar/Materials/Fonts/MI_Sidebar_FontColor_Button_Attention
     fontObject=&HeadingNow
     letterSpacing=10
     monospacedWidth=1
     size=15.75
     skewAmount=0
     typefaceFontName=74 Regular
     outlineSettings {
      bApplyOutlineToDropShadows=false
      bMiteredCorners=false
      bSeparateFillAlpha=false
      outlineMaterial=/FortUILibrary/Systems/ModularButton/MaterialInstances/MI_UIKit_TextOutline
      outlineSize=0
      outlineColor=<rgba 0 0 0 1>
     }
    }
   }
   slot {class=GridSlot, properties {row=2, padding=<ltrb 18 24 18 24>}}
  }
  VisibilitySwitcher {
   widgetClassPath=&CommonVisibilitySwitcher
   bIsVariable=true
   parent=AddFriends_Grid_Root
   widget=VisibilitySwitcher
   properties {clipping=ClipToBounds}
   slot {class=GridSlot, properties {row=3, padding=<ltrb -18 0 -18 -18>}}
  }
  WBP_Sidebar_SocialTreeList {
   widgetClassPath=/Sidebar/Panels/Social/List/WBP_Sidebar_SocialTreeList
   bIsVariable=true
   parent=VisibilitySwitcher
   widget=WBP_Sidebar_SocialTreeList
   properties {bIsFocusable=true}
   slot {class=CommonVisibilitySwitcherSlot, properties {padding=<ltrb 18 12 18 0>}}
  }
  HB_Loading {
   widgetClassPath=&HorizontalBox
   bIsVariable=true
   parent=VisibilitySwitcher
   widget=HB_Loading
   slot {class=CommonVisibilitySwitcherSlot, properties {horizontalAlignment=HAlign_Center, verticalAlignment=VAlign_Top, padding=<ltrb 0 24 0 0>}}
  }
  WBP_UIKit_Throbber {
   widgetClassPath=/FortUILibrary/UIKit/Throbber/WBP_UIKit_Throbber
   bIsVariable=true
   parent=HB_Loading
   widget=WBP_UIKit_Throbber
   slot {class=HorizontalBoxSlot, properties {padding=<ltrb 0 0 12 0>}}
  }
  Text_LoadingMore {
   widgetClassPath=&CommonTextBlock
   parent=HB_Loading
   widget=Text_LoadingMore
   properties {
    mobileFontSizeMultiplier=1.2
    text=Search in progress...
    font {
     bForceMonospaced=false
     bMaterialIsStencil=false
     fontMaterial=null
     fontObject=&HeadingNow
     letterSpacing=0
     monospacedWidth=1
     size=15.75
     skewAmount=0
     typefaceFontName=75 Medium
     outlineSettings {
      bApplyOutlineToDropShadows=false
      bMiteredCorners=false
      bSeparateFillAlpha=false
      outlineMaterial=null
      outlineSize=0
      outlineColor=<rgba 0 0 0 1>
     }
    }
    shadowOffset=<0 0>
   }
   slot {class=HorizontalBoxSlot, properties {verticalAlignment=VAlign_Center}}
  }
 ]
}
